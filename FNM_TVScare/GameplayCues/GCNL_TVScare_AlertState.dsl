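; Gameplay cue for the TVScare AI alert state that shows an animated question-mark or exclamation icon with alert sounds.
; On application or when the icon becomes visible, it animates the sprite by writing effect level, start time and duration into custom primitive data.
; On application it also attaches searching or hostile/non-hostile combat sounds to the sprite, selected through the NPC pawn's UseAlertSound flag, team affiliation and effect level.
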(event EventOnIconVisible
  (params (Target :type Actor%20Object%20Reference))
  (CallFunction|AnimateAlertIcon))

(event EventOnApplication(Cascade)
  (params (MyTarget :type Actor%20Object%20Reference) (Parameters :type Gameplay%20Cue%20Parameters%20Structure%20%28by%20ref%29) (ParticleComponents :type Array%20of%20Cascade%20Particle%20System%20Component%20Object%20References) (AudioComponents :type Array%20of%20Audio%20Component%20Object%20References) (BurstCameraShakeInstance :type Camera%20Shake%20Base%20Object%20Reference) (BurstDecalInstance :type Decal%20Actor%20Object%20Reference))
  (CallFunction|PlayAlertSound MyTarget)
  (CallFunction|AnimateAlertIcon))

(fn AnimateAlertIcon
  (bind _spritemeshcomponent (Variables|FortGameplayCueNotifyAthena_AIAlertState|GetSpriteMeshComponent))
  (Rendering|Material|SetCustomPrimitiveDataFloat _spritemeshcomponent 3 (Math|Conversions|ToFloat(Integer) (Variables|FortGameplayCueNotifyAthena_AIAlertState|GetCurrentGameplayEffectLevel))) ; IconType (QuestionMark / Exclamation)
  (Rendering|Material|SetCustomPrimitiveDataFloat _spritemeshcomponent 12 (Utilities|Time|GetGameTimeinSeconds 0)) ; BounceIn
  (Rendering|Material|SetCustomPrimitiveDataFloat _spritemeshcomponent 5 (Variables|Default|GetAnimationLength)) ; ScaleInDuration
)

(fn PlayAlertSound
  (params (TargetedActor :type Actor%20Object%20Reference))
  (bind _asbp_player_pawn_non_participant (Utilities|Casting|CastToBP_PlayerPawn_NonParticipant TargetedActor))
  (if (Variables|/Game/Athena/AI/NPCs/Base/BP_PlayerPawn_NonParticipant.BP_PlayerPawn_NonParticipant_C|GetUseAlertSound _asbp_player_pawn_non_participant)
    (switch Utilities|FlowControl|Switch|SwitchonEFortTeamAffiliation (Team|GetTeamAffiliationforActortoLocalPlayer _asbp_player_pawn_non_participant true true)
      (:Friendly
        _)
      (:Neutral
        (switch int (Variables|FortGameplayCueNotifyAthena_AIAlertState|GetCurrentGameplayEffectLevel)
          (:0
            _)
          (:1
            ;[0.3854 0.07026 0.07026] In combat with someone else
            (Audio|SpawnSoundAttached (Variables|Sound|GetNon-hostile%20Combat%20Sound) (Variables|FortGameplayCueNotifyAthena_AIAlertState|GetSpriteMeshComponent)) ; !
          )))
      (:Hostile
        (switch int (Variables|FortGameplayCueNotifyAthena_AIAlertState|GetCurrentGameplayEffectLevel)
          (:0
            _)
          (:1
            ;[0.3854 0.07026 0.07026] In hostile combat
            (Audio|SpawnSoundAttached (Variables|Sound|GetHostile%20Combat%20Sound) (Variables|FortGameplayCueNotifyAthena_AIAlertState|GetSpriteMeshComponent)) ; !
          ))))
    ;[0.4375 0.2824 0.1755] Alerted but not in combat
    (Audio|SpawnSoundAttached (Variables|Sound|GetSearchingSound) (Variables|FortGameplayCueNotifyAthena_AIAlertState|GetSpriteMeshComponent)) ; ?
  ))

(fn ConstructionScript)
