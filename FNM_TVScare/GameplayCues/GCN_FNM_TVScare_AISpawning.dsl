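; Burst gameplay cue for TV Scare AI spawning that adds a local-only proximity effect actor for the affected player.
; OnBurst runs the parent implementation, casts the cue target to a Fort player pawn and checks that it is locally controlled.
; If so, it spawns BP_PlayersInProximity_Post at the target's transform, so only that player's client creates it.
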(fn OnBurst(EitherVFX-preferthisversion)
  (params (MyTarget :type Actor%20Object%20Reference) (Parameters :type Gameplay%20Cue%20Parameters%20Structure%20%28by%20ref%29) (ParticleComponents :type Array%20of%20FXSystem%20Component%20Object%20References) (AudioComponents :type Array%20of%20Audio%20Component%20Object%20References) (BurstCameraShakeInstance :type Camera%20Shake%20Base%20Object%20Reference) (BurstDecalInstance :type Decal%20Actor%20Object%20Reference))
  (|CallParentFunction MyTarget Parameters ParticleComponents AudioComponents BurstCameraShakeInstance BurstDecalInstance)
  ;[1 1 1] Only apply on local player
  (bind _asfort_player_pawn (Utilities|Casting|CastToFortPlayerPawn MyTarget))
  (if (Pawn|IsLocallyControlled _asfort_player_pawn)
    (Game|SpawnActorfromClass 0 "/FNM_TVScare/Blueprints/BP_PlayersInProximity_Post.BP_PlayersInProximity_Post_C" (Transformation|GetActorTransform MyTarget) "Undefined" "OverrideRootScale" 0 0)
    (return)))
